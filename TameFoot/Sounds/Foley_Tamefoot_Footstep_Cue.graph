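# Sound cue for Tamefoot footstep foley, classed as player foley with its own attenuation and shared local and global foley concurrency limits.
# A Modulator at default settings varies each step, fed by a Quality Level switch over two equally weighted random pools.
# The first quality level picks from eight step waves, while the other three share a reduced pool of four of those waves.
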SoundCue /TameFoot/Sounds/Foley_Tamefoot_Footstep_Cue
root=Modulator_0
nodes [
 {id=Modulator_0, children [QualityLevel_0]}
 {id=QualityLevel_0, children [Random_2,Random_3,Random_3,Random_3]}
 {
  id=Random_2
  children [WavePlayer_12, WavePlayer_13, WavePlayer_14, WavePlayer_15, WavePlayer_16, WavePlayer_17, WavePlayer_18, WavePlayer_19]
  properties {weights [1,1,1,1,1,1,1,1]}
 }
 {id=Random_3, children [WavePlayer_20, WavePlayer_21, WavePlayer_22, WavePlayer_23], properties {weights [1,1,1,1]}}
 {id=WavePlayer_12, properties {soundWaveAssetPtr=&/Tamefoot_Step_01}}
 {id=WavePlayer_13, properties {soundWaveAssetPtr=&/Tamefoot_Step_08}}
 {id=WavePlayer_14, properties {soundWaveAssetPtr=&/Tamefoot_Step_07}}
 {id=WavePlayer_15, properties {soundWaveAssetPtr=&/Tamefoot_Step_06}}
 {id=WavePlayer_16, properties {soundWaveAssetPtr=&/Tamefoot_Step_05}}
 {id=WavePlayer_17, properties {soundWaveAssetPtr=&/Tamefoot_Step_04}}
 {id=WavePlayer_18, properties {soundWaveAssetPtr=&/Tamefoot_Step_03}}
 {id=WavePlayer_19, properties {soundWaveAssetPtr=&/Tamefoot_Step_02}}
 {id=WavePlayer_20, properties {soundWaveAssetPtr=&Tamefoot_Step_01}}
 {id=WavePlayer_21, properties {soundWaveAssetPtr=&Tamefoot_Step_08}}
 {id=WavePlayer_22, properties {soundWaveAssetPtr=&Tamefoot_Step_07}}
 {id=WavePlayer_23, properties {soundWaveAssetPtr=&Tamefoot_Step_06}}
]
properties {
 attenuationSettings=&/Foley_Attenuation_Tamefoot_Att
 maxDistance=2525
 soundClassObject=&Fort_Player_Foley
 volumeMultiplier=0.35
 concurrencySet [
  /Game/Sounds/Fort_Foley_Sounds/FoleyLibraries/Settings/Foley_Basic_Local_Conc
  /Game/Sounds/Fort_Foley_Sounds/FoleyLibraries/Settings/Foley_Basic_Global_Conc
 ]
}
